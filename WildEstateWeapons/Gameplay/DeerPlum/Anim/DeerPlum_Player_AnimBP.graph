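# Player animation blueprint for the DeerPlum weapon, providing Item Override layers for chainsaw-style equip and scooter riding.
# ItemFullBodyOverride runs a state machine covering equip, cached body-pose pass-through, scooter intro, lean-blendspace riding and falling into the scooter, then adds idle noise and pelvis twist.
# Equip masks the lower body with the cached pose, varying by movement.
# ItemFinalPoseOverride offsets the pelvis; other layers, EventGraph and variables are empty.
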AnimBlueprint /WildEstateWeapons/Gameplay/DeerPlum/Anim/DeerPlum_Player_AnimBP
target_skeleton=&Fortnite_M_Avg_Player_Skeleton
graphs [
 AnimGraph {nodes [Output Pose()]}
 EventGraph {links [], nodes []}
 ItemAimOffset {nodes [
   Input Pose!() -> (Pose, AimOffsetAlpha:double, Yaw:double, Pitch:double) {name=InPose_AimOffset}
   Input Pose!() -> (Pose) {name=InPose_PreAimOffsets}
   Output Pose() {layerGroup=Item Override}
  ]}
 ItemFallAdditive {nodes [
   Input Pose() {name=InPoseItemFallAdditive}
   Output Pose() {layerGroup=Item Override}
  ]}
 ItemFallLandAdditive {nodes [
   Input Pose() {name=InPoseItemFallLandAdditive}
   Output Pose() {layerGroup=Item Override}
  ]}
 ItemFallLevel2Override {nodes [
   Input Pose() {name=InPose_FallLevel2}
   Output Pose() {layerGroup=Item Override}
  ]}
 ItemFinalPoseOverride {nodes [
   n1 = "Transform (Modify) Bone - Bone: pelvis"!(
    ComponentPose=Local To Component(LocalPose=Input Pose() {name=InFinalPose})
    Alpha=1
    Translation_X:double=0
    Translation_Y:double=Get ScooterPelvisOffset() -> (ScooterPelvisOffset:float)
    Translation_Z:double=0
    Rotation="0,0,0"
    Scale="1,1,1"
   ) -> (Pose) {translationMode=BMM_Additive, translationSpace=BCS_ParentBoneSpace, alphaBoolBlend {blendOption=Linear}, boneToModify {boneName=pelvis}}
   Output Pose(Result=Component To Local(ComponentPose=@n1)) {layerGroup=Item Override}
  ]}
 ItemFullBodyAimOffsets {nodes [
   Input Pose!() -> (Pose, FullBody_AimOffsetAlpha:double, Aim_Yaw:double, Aim_Pitch:double) {name=InPoseFullBodyAimOffsets}
   Input Pose!() -> (Pose) {name=InPosePreFullBodyAimOffsets}
   Output Pose() {layerGroup=Item Override}
  ]}
 ItemFullBodyOverride {nodes [
   n1 = "Sequence Player (additive) 'AS_Chainsaw_idle_Noise'"() {sequence=/ChainsawGameplay/Assets/Animations/AS_Chainsaw_idle_Noise}
   AnimGraphNode_StateMachine_0 = Chainsaw Scooter/Equip SM() {is=title AnimGraphNode_StateMachine}
   n2 = "Transform (Modify) Bone - Bone: pelvis"!(
    ComponentPose=Local To Component(LocalPose=Apply Mesh Space Additive!(Base:Pose Link Structure=@AnimGraphNode_StateMachine_0, Additive:Pose Link Structure=@n1, Alpha:float=0.4) -> (Pose:Pose Link Structure) {is=title AnimGraphNode_ApplyMeshSpaceAdditive, alphaBoolBlend {blendOption=Linear}})
    Alpha=1
    Translation:Vector="0,0,0"
    Rotation_Roll:float=0
    Rotation_Pitch:float=0
    Rotation_Yaw:float=Get MeleeTwist() -> (MeleeTwist:float)
    Scale="1,1,1"
   ) -> (Pose) {rotationMode=BMM_Additive, rotationSpace=BCS_ParentBoneSpace, alphaBoolBlend {blendOption=Linear}, boneToModify {boneName=pelvis}}
   Output Pose(Result=Component To Local(ComponentPose=@n2)) {layerGroup=Item Override}
   Save cached pose 'FullBody Pose'(Pose=Input Pose() {name=InPoseFullBody}) {cacheName=FullBody Pose}
  ]}
 Chainsaw Scooter/Equip SM {
  entry=Equip
  path=ItemFullBodyOverride.AnimGraphNode_StateMachine_0.Chainsaw Scooter/Equip SM
  ids=id
  nodes [
   To FallingIntoScooter() {is=title AnimStateAliasNode, also {aliases [Pass Through]}}
   To ScooterFalling() {is=title AnimStateAliasNode, also {aliases [Pass Through, ScooterIntro, Scooting]}}
   To Pass Through() {
    is=title AnimStateAliasNode
    also {aliases [FallingIntoScooter Land, FallingIntoScooter Loop, FallingIntoScooter Start, ScooterIntro, Scooting]}
   }
   AnimStateNode_0 = Equip() {is=title AnimStateNode}
   AnimStateNode_10 = Pass Through() {
    is=title AnimStateNode
    stateEntered {guid=#1}
    stateFullyBlended {guid=#2}
    stateLeft {guid=#3}
   }
   AnimStateNode_11 = Scooting() {
    is=title AnimStateNode
    stateEntered {guid=#4}
    stateFullyBlended {guid=#5}
    stateLeft {guid=#6}
   }
   AnimStateNode_15 = FallingIntoScooter Start() {
    is=title AnimStateNode
    stateEntered {guid=#7}
    stateFullyBlended {guid=#8}
    stateLeft {guid=#9}
   }
   AnimStateNode_16 = FallingIntoScooter Loop() {
    is=title AnimStateNode
    stateEntered {guid=#10}
    stateFullyBlended {guid=#11}
    stateLeft {guid=#12}
   }
   AnimStateNode_17 = FallingIntoScooter Land() {
    is=title AnimStateNode
    stateEntered {guid=#13}
    stateFullyBlended {guid=#14}
    stateLeft {guid=#15}
   }
   AnimStateNode_9 = ScooterIntro() {is=title AnimStateNode}
   AnimStateTransitionNode_0 = AnimStateTransitionNode() {crossfadeDuration=0.1}
   AnimStateTransitionNode_1 = AnimStateTransitionNode() {bAutomaticRuleBasedOnSequencePlayerInState=true, crossfadeDuration=0.5}
   AnimStateTransitionNode_12 = AnimStateTransitionNode() {crossfadeDuration=0.1}
   AnimStateTransitionNode_13 = AnimStateTransitionNode() {bAutomaticRuleBasedOnSequencePlayerInState=true, crossfadeDuration=0.3}
   AnimStateTransitionNode_14 = AnimStateTransitionNode() {
    crossfadeDuration=0.05
    transitionEnd {guid=#16}
    transitionInterrupt {guid=#17}
    transitionStart {guid=#18}
   }
   AnimStateTransitionNode_17 = AnimStateTransitionNode() {crossfadeDuration=0.3}
   AnimStateTransitionNode_25 = AnimStateTransitionNode() {
    crossfadeDuration=0.3
    transitionEnd {guid=#19}
    transitionInterrupt {guid=#20}
    transitionStart {guid=#21}
   }
   AnimStateTransitionNode_26 = AnimStateTransitionNode() {
    bAutomaticRuleBasedOnSequencePlayerInState=true
    crossfadeDuration=0.4
    transitionEnd {guid=#22}
    transitionInterrupt {guid=#23}
    transitionStart {guid=#24}
   }
   AnimStateTransitionNode_27 = AnimStateTransitionNode() {
    crossfadeDuration=0.05
    transitionEnd {guid=#25}
    transitionInterrupt {guid=#26}
    transitionStart {guid=#27}
   }
   AnimStateTransitionNode_28 = AnimStateTransitionNode() {
    bAutomaticRuleBasedOnSequencePlayerInState=true
    crossfadeDuration=0.3
    transitionEnd {guid=#28}
    transitionInterrupt {guid=#29}
    transitionStart {guid=#30}
   }
   AnimStateTransitionNode_31 = AnimStateTransitionNode() {
    transitionEnd {guid=#31}
    transitionInterrupt {guid=#32}
    transitionStart {guid=#33}
   }
   AnimStateTransitionNode_7 = AnimStateTransitionNode() {crossfadeDuration=0}
  ]
  links [
   {
    dst=Pass Through
    rule=ItemFullBodyOverride.AnimGraphNode_StateMachine_0.Chainsaw Scooter/Equip SM.AnimStateTransitionNode_1.Transition
    src=Equip
    via=AnimStateTransitionNode_1
   }
   {
    dst=Pass Through
    rule=ItemFullBodyOverride.AnimGraphNode_StateMachine_0.Chainsaw Scooter/Equip SM.AnimStateTransitionNode_7.Transition
    src=Equip
    via=AnimStateTransitionNode_7
   }
   {
    dst=Scooting
    rule=ItemFullBodyOverride.AnimGraphNode_StateMachine_0.Chainsaw Scooter/Equip SM.AnimStateTransitionNode_28.AnimationTransitionGraph_1
    src=FallingIntoScooter Land
    via=AnimStateTransitionNode_28
   }
   {
    dst=FallingIntoScooter Land
    rule=ItemFullBodyOverride.AnimGraphNode_StateMachine_0.Chainsaw Scooter/Equip SM.AnimStateTransitionNode_27.AnimationTransitionGraph_1
    src=FallingIntoScooter Loop
    via=AnimStateTransitionNode_27
   }
   {
    dst=FallingIntoScooter Start
    rule=ItemFullBodyOverride.AnimGraphNode_StateMachine_0.Chainsaw Scooter/Equip SM.AnimStateTransitionNode_0.Transition
    src=FallingIntoScooter Loop
    via=AnimStateTransitionNode_0
   }
   {
    dst=FallingIntoScooter Land
    rule=ItemFullBodyOverride.AnimGraphNode_StateMachine_0.Chainsaw Scooter/Equip SM.AnimStateTransitionNode_14.Transition
    src=FallingIntoScooter Start
    via=AnimStateTransitionNode_14
   }
   {
    dst=FallingIntoScooter Loop
    rule=ItemFullBodyOverride.AnimGraphNode_StateMachine_0.Chainsaw Scooter/Equip SM.AnimStateTransitionNode_26.AnimationTransitionGraph_1
    src=FallingIntoScooter Start
    via=AnimStateTransitionNode_26
   }
   {
    dst=ScooterIntro
    rule=ItemFullBodyOverride.AnimGraphNode_StateMachine_0.Chainsaw Scooter/Equip SM.AnimStateTransitionNode_12.AnimationTransitionGraph_0
    src=Pass Through
    via=AnimStateTransitionNode_12
   }
   {
    dst=Scooting
    rule=ItemFullBodyOverride.AnimGraphNode_StateMachine_0.Chainsaw Scooter/Equip SM.AnimStateTransitionNode_13.AnimationTransitionGraph_0
    src=ScooterIntro
    via=AnimStateTransitionNode_13
   }
   {
    dst=FallingIntoScooter Start
    rule=ItemFullBodyOverride.AnimGraphNode_StateMachine_0.Chainsaw Scooter/Equip SM.AnimStateTransitionNode_25.AnimationTransitionGraph_1
    src=To FallingIntoScooter
    via=AnimStateTransitionNode_25
   }
   {
    dst=Pass Through
    rule=ItemFullBodyOverride.AnimGraphNode_StateMachine_0.Chainsaw Scooter/Equip SM.AnimStateTransitionNode_17.Transition
    src=To Pass Through
    via=AnimStateTransitionNode_17
   }
   {
    dst=FallingIntoScooter Start
    rule=ItemFullBodyOverride.AnimGraphNode_StateMachine_0.Chainsaw Scooter/Equip SM.AnimStateTransitionNode_31.Transition
    src=To ScooterFalling
    via=AnimStateTransitionNode_31
   }
  ]
 }
 Equip {
  path=ItemFullBodyOverride.AnimGraphNode_StateMachine_0.Chainsaw Scooter/Equip SM.AnimStateNode_0.Equip
  nodes [
   n1 = Sequence Player 'AS_Chainsaw_run_equip'() {bLoopAnimation=false, groupName=equip, method=SyncGroup, sequence=/ChainsawGameplay/Assets/Animations/AS_Chainsaw_run_equip}
   n2 = Layered blend per bone(BasePose=@n1, BlendPoses_0=Use cached pose 'FullBody Pose'()) {
    bMeshSpaceRotationBlend=true
    blendMode=BlendMask
    blendMasks [&Fortnite_M_Avg_Player_Skeleton:LowerBody]
    layerSetup []
   }
   n3 = Sequence Player 'AS_ChainSaw_Idle_Equip'() {bLoopAnimation=false, groupName=equip, method=SyncGroup, sequence=/ChainsawGameplay/Assets/Animations/AS_ChainSaw_Idle_Equip}
   n4 = Layered blend per bone(BasePose=@n3, BlendPoses_0=Use cached pose 'FullBody Pose'()) {
    bMeshSpaceRotationBlend=true
    blendMode=BlendMask
    blendMasks [&Fortnite_M_Avg_Player_Skeleton:LowerBody]
    layerSetup []
   }
   n5 = Sequence Player 'AS_ChainSaw_Slide_Equip'() {bLoopAnimation=false, groupName=equip, method=SyncGroup, sequence=/ChainsawGameplay/Assets/Animations/AS_ChainSaw_Slide_Equip}
   n6 = Layered blend per bone(BasePose=@n5, BlendPoses_0=Use cached pose 'FullBody Pose'()) {
    bMeshSpaceRotationBlend=true
    blendMode=BlendMask
    blendMasks [&Fortnite_M_Avg_Player_Skeleton:LowerBody]
    layerSetup []
   }
   n7 = Blend Poses by bool(
    BlendPose_0=@n6
    BlendPose_1=Blend Poses by bool(BlendPose_0=@n2, BlendPose_1=@n4, bActiveValue=Get bIsSprinting()) {blendType=Linear}
    bActiveValue=Get bIsSlidingMovementMode() -> (bIsSlidingMovementMode:bool)
   ) {blendType=Linear}
   n8 = Sequence Player 'AS_ChainSaw_Slide_Equip'() {bLoopAnimation=false, groupName=equip, method=SyncGroup, sequence=&AS_ChainSaw_Slide_Equip}
   n9 = Layered blend per bone(BasePose=@n8, BlendPoses_0=Use cached pose 'FullBody Pose'()) {
    bMeshSpaceRotationBlend=true
    blendMode=BlendMask
    blendMasks [&Fortnite_M_Avg_Player_Skeleton:LowerBody]
    layerSetup []
   }
   n10 = Get bIsFalling()
   n11 = "Transform (Modify) Bone - Bone: pelvis"(
    ComponentPose=Local To Component(LocalPose=Blend Poses by bool(BlendPose_0=@n9, BlendPose_1=@n7, bActiveValue=@n10) {blendType=Linear})
    Alpha=1
    Translation="0,0,0"
    Rotation_Roll=0
    Rotation_Pitch=0
    Rotation_Yaw=Get CustomMeleeTwist() -> (CustomMeleeTwist:float)
    Scale="1,1,1"
   ) {rotationMode=BMM_Additive, rotationSpace=BCS_ParentBoneSpace, alphaBoolBlend {blendOption=Linear}, boneToModify {boneName=pelvis}}
   n12 = Sequence Player 'AS_ChainSaw_Idle_Equip'() {bLoopAnimation=false, groupName=equip, method=SyncGroup, sequence=&AS_ChainSaw_Idle_Equip}
   Output Animation Pose(Result=Blend Poses by bool(BlendPose_0=Component To Local(ComponentPose=@n11), BlendPose_1=@n12, bActiveValue=Get bIsUpperBodyEquip() -> (bIsUpperBodyEquip:bool)) {blendType=Linear})
  ]
 }
 Pass Through {
  path=ItemFullBodyOverride.AnimGraphNode_StateMachine_0.Chainsaw Scooter/Equip SM.AnimStateNode_10.Pass Through
  nodes [Output Animation Pose(Result=Use cached pose 'FullBody Pose'())]
 }
 Scooting {
  path=ItemFullBodyOverride.AnimGraphNode_StateMachine_0.Chainsaw Scooter/Equip SM.AnimStateNode_11.Scooting
  nodes [Output Animation Pose(Result=Apply Mesh Space Additive(
     Base=Blendspace Player 'DeerPlum_Scooter_BS_Leans'(X=0, Y=0) {blendSpace=&/DeerPlum_Scooter_BS_Leans}
     Additive="Sequence Player (additive) 'AS_DeerPlum_scooter_N_Additive'"() {sequence=&/AS_DeerPlum_scooter_N_Additive}
     Alpha=1
    ) {alphaBoolBlend {blendOption=Linear}})]
 }
 FallingIntoScooter Start {
  path=ItemFullBodyOverride.AnimGraphNode_StateMachine_0.Chainsaw Scooter/Equip SM.AnimStateNode_15.FallingIntoScooter Start
  nodes [
   n1 = Sequence Player 'AS_Chainsaw_Scooter_Intro_InAir'() {bLoopAnimation=false, groupName=ScooterIntro, method=SyncGroup, sequence=/ChainsawGameplay/Assets/Animations/AS_Chainsaw_Scooter_Intro_InAir}
   Output Animation Pose(Result=@n1)
   n2 = Sequence Player 'AS_Chainsaw_Scooter_Intro_InAir'() {bLoopAnimation=false, groupName=ScooterIntro, method=SyncGroup, sequence=&AS_Chainsaw_Scooter_Intro_InAir}
   n3 = Layered blend per bone(BasePose=@n2, BlendPoses_0=Use cached pose 'FullBody Pose'()) {blendMode=BlendMask, blendMasks [&Fortnite_M_Avg_Player_Skeleton:LowerBody], layerSetup []}
   n4 = Sequence Player 'AS_Chainsaw_Scooter_Intro_InAir'() {bLoopAnimation=false, groupName=ScooterIntro, method=SyncGroup, sequence=&AS_Chainsaw_Scooter_Intro_InAir}
   Two Way Blend(A=@n4, B=@n3, Alpha=0) {blendNode {alphaBoolBlend {blendOption=Linear}}}
  ]
 }
 FallingIntoScooter Loop {
  path=ItemFullBodyOverride.AnimGraphNode_StateMachine_0.Chainsaw Scooter/Equip SM.AnimStateNode_16.FallingIntoScooter Loop
  nodes [Output Animation Pose(Result=Sequence Player 'AS_Chainsaw_Scooter_Intro_fall'() {sequence=/ChainsawGameplay/Assets/Animations/AS_Chainsaw_Scooter_Intro_fall})]
 }
 FallingIntoScooter Land {
  path=ItemFullBodyOverride.AnimGraphNode_StateMachine_0.Chainsaw Scooter/Equip SM.AnimStateNode_17.FallingIntoScooter Land
  nodes [Output Animation Pose(Result=Sequence Player 'AS_DeerPlum_Scooter_Intro_Land'() {bLoopAnimation=false, playRate=0.5, sequence=&/AS_DeerPlum_Scooter_Intro_Land})]
 }
 ScooterIntro {
  path=ItemFullBodyOverride.AnimGraphNode_StateMachine_0.Chainsaw Scooter/Equip SM.AnimStateNode_9.ScooterIntro
  nodes [Output Animation Pose(Result=Sequence Player 'AS_DeerPlum_Scooter_Intro'() {bLoopAnimation=false, sequence=&/AS_DeerPlum_Scooter_Intro})]
 }
 Transition {
  path=ItemFullBodyOverride.AnimGraphNode_StateMachine_0.Chainsaw Scooter/Equip SM.AnimStateTransitionNode_0.Transition
  nodes [Result(bCanEnterTransition=Get bIsJumping())]
 }
 Transition {path=ItemFullBodyOverride.AnimGraphNode_StateMachine_0.Chainsaw Scooter/Equip SM.AnimStateTransitionNode_1.Transition, nodes [Result()]}
 AnimationTransitionGraph_0 {
  path=ItemFullBodyOverride.AnimGraphNode_StateMachine_0.Chainsaw Scooter/Equip SM.AnimStateTransitionNode_12.AnimationTransitionGraph_0
  nodes [Result(bCanEnterTransition=Get bIsChainsawScooter() -> (bIsChainsawScooter:bool))]
 }
 AnimationTransitionGraph_0 {
  path=ItemFullBodyOverride.AnimGraphNode_StateMachine_0.Chainsaw Scooter/Equip SM.AnimStateTransitionNode_13.AnimationTransitionGraph_0
  nodes [Result()]
 }
 Transition {
  path=ItemFullBodyOverride.AnimGraphNode_StateMachine_0.Chainsaw Scooter/Equip SM.AnimStateTransitionNode_14.Transition
  nodes [Result(bCanEnterTransition=Get bIsChainsawScooter())]
 }
 Transition {
  path=ItemFullBodyOverride.AnimGraphNode_StateMachine_0.Chainsaw Scooter/Equip SM.AnimStateTransitionNode_17.Transition
  nodes [Result(bCanEnterTransition=Get bIsNotChainsawScooter() -> (bIsNotChainsawScooter:bool))]
 }
 AnimationTransitionGraph_1 {
  path=ItemFullBodyOverride.AnimGraphNode_StateMachine_0.Chainsaw Scooter/Equip SM.AnimStateTransitionNode_25.AnimationTransitionGraph_1
  nodes [Result(bCanEnterTransition=Get bIsFallingIntoScooter() -> (bIsFallingIntoScooter:bool))]
 }
 AnimationTransitionGraph_1 {
  path=ItemFullBodyOverride.AnimGraphNode_StateMachine_0.Chainsaw Scooter/Equip SM.AnimStateTransitionNode_26.AnimationTransitionGraph_1
  nodes [Result()]
 }
 AnimationTransitionGraph_1 {
  path=ItemFullBodyOverride.AnimGraphNode_StateMachine_0.Chainsaw Scooter/Equip SM.AnimStateTransitionNode_27.AnimationTransitionGraph_1
  nodes [Result(bCanEnterTransition=Get bIsChainsawScooter())]
 }
 AnimationTransitionGraph_1 {
  path=ItemFullBodyOverride.AnimGraphNode_StateMachine_0.Chainsaw Scooter/Equip SM.AnimStateTransitionNode_28.AnimationTransitionGraph_1
  nodes [Result()]
 }
 Transition {
  path=ItemFullBodyOverride.AnimGraphNode_StateMachine_0.Chainsaw Scooter/Equip SM.AnimStateTransitionNode_31.Transition
  nodes [Result(bCanEnterTransition=Get bIsScooterFalling() -> (bIsScooterFalling:bool))]
 }
 Transition {
  path=ItemFullBodyOverride.AnimGraphNode_StateMachine_0.Chainsaw Scooter/Equip SM.AnimStateTransitionNode_7.Transition
  nodes [Result(bCanEnterTransition=Get bIsNotEquip() -> (bIsNotEquip:bool))]
 }
 ItemIdleOverride {nodes [
   Input Pose() {name=InPose_Idle}
   Output Pose() {layerGroup=Item Override}
  ]}
 ItemJetPackJumpAdditive {nodes [Output Pose() {layerGroup=Item Override}]}
 ItemJetPackStartAdditive {nodes [Output Pose() {layerGroup=Item Override}]}
 ItemJumpLoopAdditive {nodes [
   Input Pose() {name=InPoseItemJumpLoopAdditive}
   Output Pose() {layerGroup=Item Override}
  ]}
 ItemJumpOverride {nodes [
   Input Pose() {name=InPoseJump}
   Output Pose() {layerGroup=Item Override}
  ]}
 ItemJumpUpAdditive {nodes [
   Input Pose() {name=InPoseItemJumpUpAdditive}
   Output Pose() {layerGroup=Item Override}
  ]}
 ItemPreAimOffsets {nodes [
   Input Pose() {name=InPosePreAimOffsets}
   Output Pose() {layerGroup=Item Override}
  ]}
 ItemPreIK {nodes [
   Input Pose() {name=InPose_PreIK}
   Output Pose() {layerGroup=Item Override}
  ]}
 ItemRightHandIK {nodes [
   Input Pose() {name=InPosePreRightHandIK}
   Input Pose() {name=InPosePostRightHandIK}
   Output Pose() {layerGroup=Item Override}
  ]}
 ItemSkeletalControl {nodes [
   Input Pose()
   Output Pose() {layerGroup=Item Override}
  ]}
 ItemSwimJumpFallLoop {nodes [Output Pose() {layerGroup=Item Override}]}
 ItemSwimJumpSurfaceEnd {nodes [Output Pose() {layerGroup=Item Override}]}
 ItemSwimJumpSurfaceLoop {nodes [Output Pose() {layerGroup=Item Override}]}
 ItemUpperBody {nodes [
   Input Pose!() -> (Pose, InputParam:Fort Anim Input Adjusted Aim Structure) {name=InPoseUpperBody}
   Output Pose() {layerGroup=Item Override}
  ]}
 ItemZipLineStartAdditive {nodes [Output Pose() {layerGroup=Item Override}]}
]
variables []
